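# Material function for glTF material import that computes the sheen layer's color and roughness.
# SheenColor multiplies the color texture's RGB by SheenColorFactor, and SheenRoughness multiplies the linear roughness texture's alpha by SheenRoughnessFactor.
# Each texture takes its own UVs and explicit DDX/DDY derivatives and samples through the shared clamp sampler.
# Texture inputs default to white, leaving the factors unchanged.
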MaterialFunction /InterchangeAssets/gltf/MaterialFunctions/MF_Sheen
graph {nodes [
  n1 = FunctionInput() {bUsePreviewValueAsDefault=true, inputName=SheenColorTexture_UVs, inputType=FunctionInput_Vector2, sortPriority=4}
  n2 = FunctionInput(Preview=TextureObject() {texture=/InterchangeAssets/gltf/Textures/T_White_srgb}) {bUsePreviewValueAsDefault=true, inputName=SheenColorTexture, inputType=FunctionInput_Texture2D}
  n3 = FunctionInput() {bUsePreviewValueAsDefault=true, inputName=SheenColorTexture_DDX, inputType=FunctionInput_Vector2, sortPriority=5}
  n4 = FunctionInput() {bUsePreviewValueAsDefault=true, inputName=SheenColorTexture_DDY, inputType=FunctionInput_Vector2, sortPriority=6}
  FunctionOutput(None=Multiply(
    A=TextureSample!(UVs=@n1, Tex=@n2, "DDX(UVs)"=@n3, "DDY(UVs)"=@n4, Apply View MipBias) {mipValueMode=TMVM_Derivative, samplerSource=SSM_Clamp_WorldGroupSettings}.RGB
    B=FunctionInput() {bUsePreviewValueAsDefault=true, inputName=SheenColorFactor, sortPriority=2}
   )) {outputName=SheenColor}
  n5 = FunctionInput() {bUsePreviewValueAsDefault=true, inputName=SheenRoughnessTexture_UVs, inputType=FunctionInput_Vector2, sortPriority=7}
  n6 = FunctionInput(Preview=TextureObject() {samplerType=SAMPLERTYPE_LinearColor, texture=/InterchangeAssets/gltf/Textures/T_White_Linear}) {bUsePreviewValueAsDefault=true, inputName=SheenRoughnessTexture, inputType=FunctionInput_Texture2D, sortPriority=1}
  n7 = FunctionInput() {bUsePreviewValueAsDefault=true, inputName=SheenRoughnessTexture_DDX, inputType=FunctionInput_Vector2, sortPriority=8}
  n8 = FunctionInput() {bUsePreviewValueAsDefault=true, inputName=SheenRoughnessTexture_DDY, inputType=FunctionInput_Vector2, sortPriority=9}
  FunctionOutput(None=Multiply(
    A=TextureSample(UVs=@n5, Tex=@n6, "DDX(UVs)"=@n7, "DDY(UVs)"=@n8) {mipValueMode=TMVM_Derivative, samplerSource=SSM_Clamp_WorldGroupSettings, samplerType=SAMPLERTYPE_LinearColor}.A
    B=FunctionInput() {bUsePreviewValueAsDefault=true, inputName=SheenRoughnessFactor, inputType=FunctionInput_Scalar, sortPriority=3}
   )) {outputName=SheenRoughness, sortPriority=1}
 ]}
