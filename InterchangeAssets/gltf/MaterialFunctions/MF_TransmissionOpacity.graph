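# Material function in the Interchange glTF set that computes an Opacity output for transmissive materials.
# A Fresnel-like facing term from tangent-space normals is maxed with a power of Roughness.
# That is maxed with a distance term: scene depth at DistortedUVs minus pixel depth, converted to view-ray distances, normalized by object size, inverted and scaled by inverse roughness.
# Mobile feature levels replace the distance term with a small constant.
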MaterialFunction /InterchangeAssets/gltf/MaterialFunctions/MF_TransmissionOpacity
graph {
 properties {previewBlendMode=BLEND_Translucent}
 nodes [
  n1 = FunctionInput(Preview=Constant3Vector() {constant=<rgba 0 0 1 0>}) {bUsePreviewValueAsDefault=true, inputName=Normals TS}
  n2 = Abs(None=DotProduct(A=Transform(None=@n1), B=CameraVectorWS()) {desc=Facing Value})
  n3 = FunctionInput(Preview=Constant()) {bUsePreviewValueAsDefault=true, inputName=Roughness, inputType=FunctionInput_Scalar}
  n4 = Max(A=Power(Base=OneMinus(None=@n2)) {constExponent=5}, B=Power(Base=@n3) {constExponent=0.2}) {desc=Roughness contribution}
  n5 = Constant2Vector()
  n6 = DotProduct(
   A=Multiply(A=CameraVectorWS()) {constB=-1}
   B=MaterialFunctionCall() {materialFunction=&CameraDirectionVector}.Result
  )
  n7 = Divide(
   A=SceneDepth(UVs=FunctionInput(Preview=@n5) {bUsePreviewValueAsDefault=true, inputName=DistortedUVs, inputType=FunctionInput_Vector2}).R
   B=@n6
  )
  n8 = Divide(A=PixelDepth().R, B=@n6)
  n9 = Multiply(
   A=MaterialFunctionCall() {materialFunction=&ObjectLocalBounds}.Local Bounds Size
   B=MaterialFunctionCall() {materialFunction=&ObjectScale}.Scale XYZ
  )
  n10 = Saturate(None=Divide(
    A=Subtract(A=@n7, B=@n8)
    B=MaterialFunctionCall(f3=@n9) {materialFunction=&AddComponents}.f3
   ) {constB=500, desc=Object Size})
  n11 = Constant() {r=0.01}
  n12 = FeatureLevelSwitch(Default=Multiply(A=OneMinus(None=@n3), B=OneMinus(None=@n10)), ES2=@n11, ES3_1=@n11)
  FunctionOutput(None=Saturate(None=Max(A=@n4, B=@n12) {constB=0.2, desc=Distance Contribution})) {outputName=Opacity}
  ObjectRadius()
 ]
}
