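# Material function from the Interchange glTF assets that rotates a UV coordinate around a center point, defaulting to 0.5, 0.5.
# It takes a precomputed sine/cosine pair, defaulting to no rotation, and applies the 2D rotation as two dot products.
# V is negated before and after rotating to make the direction clockwise.
# RotationCenter is subtracted and re-added in that flipped space, without being flipped itself.
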MaterialFunction /InterchangeAssets/gltf/MaterialFunctions/MF_RotateV2
graph {nodes [
  n1 = FunctionInput() {
   bUsePreviewValueAsDefault=true
   description=Enter a number to be used as the center point of rotation.
   inputName=RotationCenter
   inputType=FunctionInput_Vector2
   previewValue=<0.5 0.5 0 1>
  }
  n2 = Add(
   A=Multiply(A=@n1, B=Constant() {r=-1})
   B=Multiply(A=FunctionInput() {inputName=UV, inputType=FunctionInput_Vector2}, B=Constant2Vector() {g=-1, r=1}) {desc=clockwise}
  )
  n3 = FunctionInput(Preview=Constant2Vector() {g=1}) {bUsePreviewValueAsDefault=true, inputName=SinCosAngle, inputType=FunctionInput_Vector2, previewValue=<0 1 0 0>}
  n4 = ComponentMask(None=@n3) {g=true}
  n5 = Multiply(A=ComponentMask(None=@n3) {r=true}) {constB=-1}
  n6 = AppendVector(A=DotProduct(A=@n2, B=AppendVector(A=@n4, B=@n5)), B=DotProduct(A=@n2, B=@n3))
  FunctionOutput(None=Multiply(A=Add(A=@n1, B=@n6), B=Constant2Vector() {g=-1, r=1}) {desc=clockwise})
 ]}
