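# MF_ScaleNormal is a material function in the Interchange glTF assets that adjusts the strength of a normal vector by a scalar.
# It multiplies the Normal input by a vector built from Scale in X and Y and a constant one in Z, then normalizes the product.
# Only the X and Y components are scaled before normalization, so Scale controls how far the normal tilts away from Z.
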MaterialFunction /InterchangeAssets/gltf/MaterialFunctions/MF_ScaleNormal
graph {nodes [
  n1 = FunctionInput() {inputName=Normal, previewValue=<0 0 1 1>}
  n2 = FunctionInput() {inputName=Scale, inputType=FunctionInput_Scalar, sortPriority=1, previewValue=<1 0 0 1>}
  n3 = MaterialFunctionCall(X=@n2, Y=@n2, Z=Constant() {r=1}) {materialFunction=&MakeFloat3}
  FunctionOutput(None=Normalize(VectorInput=Multiply(A=@n1, B=@n3.Result)))
 ]}
