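# glTF material function that returns refracted, optionally blurred scene color and depth behind a surface, as for rough glass.
# View-space refraction from normals and IOR offsets screen UVs; BlurIntensity, field of view and distance behind scale the blur.
# A Custom HLSL loop takes temporally dithered, luminance-weighted samples, rejecting foreground depths and shrinking the radius.
# With ray tracing, the color output switches to a single SceneColor lookup.
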MaterialFunction /InterchangeAssets/gltf/MaterialFunctions/MF_Temporal_Blur
graph {
 properties {bExposeToLibrary=true, previewBlendMode=BLEND_Translucent}
 nodes [
  MaterialExpressionNamedRerouteUsage_15 = NamedRerouteUsage()
  n1 = Custom!(uv=@MaterialExpressionNamedRerouteUsage_15) {
   code="return BufferUVToViewportUV(uv);"
   description=BufferUVToViewportUV
   outputType=CMOT_Float2
   inputs [{inputName=uv, input {expression=&self:MaterialExpressionNamedRerouteUsage_15}}]
  }
  n2 = NamedRerouteUsage()
  MaterialExpressionLinearInterpolate_0 = LinearInterpolate(Alpha=Saturate(None=Ceil(None=@n2))) {constA=1, constB=8}
  MaterialExpressionNamedRerouteUsage_5 = NamedRerouteUsage()
  MaterialExpressionNamedRerouteUsage_16 = NamedRerouteUsage()
  MaterialExpressionNamedRerouteUsage_19 = NamedRerouteUsage()
  MaterialExpressionNamedRerouteUsage_18 = NamedRerouteUsage()
  MaterialExpressionNamedRerouteUsage_21 = NamedRerouteUsage()
  n3 = Custom!(
   SampleCount=@MaterialExpressionLinearInterpolate_0
   BlurRadius=@MaterialExpressionNamedRerouteUsage_5
   TemporalSample=@MaterialExpressionNamedRerouteUsage_16
   RefractionOffset=@MaterialExpressionNamedRerouteUsage_19
   DistanceBehind=@MaterialExpressionNamedRerouteUsage_18
   cosdepth=@MaterialExpressionNamedRerouteUsage_21
  ) {
   code="float2 BaseUV = MaterialFloat2(ScreenAlignedPosition(Parameters.ScreenPosition).xy);\r\nfloat2 ViewRectMinUV = (View.ViewRectMin.xy + 1.f) * View.BufferSizeAndInvSize.zw;\r\nfloat2 ViewRectMaxUV = (View.ViewRectMin.xy + View.ViewSizeAndInvSize.xy - 1.f) * View.BufferSizeAndInvSize.zw;\r\nfloat3 CummulativeColor = float3(0,0,0);\r\nfloat CummulativeDepth = 0;\r\nfloat pDepth = GetPixelDepth(Parameters)/cosdepth;\r\nint i=0;\r\nfloat2 TemporalOffset = float2(0,0);\r\n\r\nint CummulativeSamples = 0;\r\n[loop]\r\nwhile (i < (int)SampleCount)\r\n{\r\n\t// Calculate UVs\r\n\t{\r\n\t\tfloat Angle = TemporalSample;\r\n\t\t// periodic\r\n\t\tAngle *= 6.283178f;\r\n\t\tfloat2 RotationCenter = float2(0.5f, 0.5f);\r\n\t\tfloat2 pos = (RotationCenter*-1.0f) + float2(1+(i/SampleCount),0);\r\n\r\n\t\tfloat s;\r\n\t\tfloat c;\r\n\t\tsincos(Angle,s,c);\r\n\t\t\r\n\t\tfloat d1 = dot(pos, float2(c,s*-1.0f));\r\n\t\tfloat d2 = dot(pos, float2(s,c));\r\n\t\tTemporalOffset = float2(d1,d2) + RotationCenter;\r\n\t}\r\n\tTemporalOffset = TemporalOffset - 0.5f;\r\n\tfloat2 SampleUV = BaseUV + RefractionOffset + (BlurRadius*TemporalOffset*DistanceBehind);\r\n\tSampleUV = clamp(SampleUV, ViewRectMinUV, ViewRectMaxUV);\r\n\t\r\n\t// Discard sample based on depth\r\n\tfloat sDepth = CalcSceneDepth(SampleUV)/cosdepth;\r\n\tif (pDepth < sDepth)\r\n\t{\r\n\t\tfloat3 SampleColor = DecodeSceneColorForMaterialNode(SampleUV).rgb;\r\n\t\tSampleColor *= rcp(1 + Luminance(SampleColor));\r\n\t\tCummulativeColor += SampleColor;\r\n\t\tCummulativeDepth += sDepth;\r\n\t\tCummulativeSamples++;\r\n\t}\r\n\telse\r\n\t{\r\n\t\t// try smaller radius next time\r\n\t\tBlurRadius *= 0.5f;\r\n\t\tRefractionOffset = RefractionOffset - (RefractionOffset*0.5f);\r\n\t\tDistanceBehind=1.0f;\r\n\t}\r\n\ti++;\r\n}\r\n\r\nif(CummulativeSamples == 0)\r\n{\r\n\tBluredDepth = CalcSceneDepth(BaseUV)/cosdepth;\r\n\treturn DecodeSceneColorForMaterialNode(BaseUV).rgb;\r\n}\r\nelse\r\n{\r\n\tBluredDepth = CummulativeDepth / CummulativeSamples;\r\n\tCummulativeColor = CummulativeColor / CummulativeSamples;\r\n\tCummulativeColor *= rcp(1 - Luminance(CummulativeColor));\r\n\treturn CummulativeColor;\r\n}\r\n"
   description=TemporalBlur
   additionalOutputs [{outputName=BluredDepth, outputType=CMOT_Float1}]
   inputs [
    {inputName=SampleCount, input {expression=&self:MaterialExpressionLinearInterpolate_0}}
    {inputName=BlurRadius, input {expression=&self:MaterialExpressionNamedRerouteUsage_5}}
    {
     inputName=TemporalSample
     input {expression=&self:MaterialExpressionNamedRerouteUsage_16}
    }
    {
     inputName=RefractionOffset
     input {expression=&self:MaterialExpressionNamedRerouteUsage_19}
    }
    {
     inputName=DistanceBehind
     input {expression=&self:MaterialExpressionNamedRerouteUsage_18}
    }
    {inputName=cosdepth, input {expression=&self:MaterialExpressionNamedRerouteUsage_21}}
   ]
  }
  FunctionOutput(None=RayTracingQualitySwitch(Normal=@n3.return, RayTraced=SceneColor!(OffsetFraction=@n1) {desc=Serves also as fix for DecodeSceneColor, inputMode=OffsetFraction}.RGB)) {outputName=Blured Scene}
  FunctionOutput(None=@n3.BluredDepth) {outputName=BluredDepth, sortPriority=1}
  n4 = Constant() {r=1.5}
  NamedRerouteDeclaration(None=Clamp(None=FunctionInput(Preview=@n4) {inputName=ior, inputType=FunctionInput_Scalar}) {maxDefault=2, minDefault=1}) {name=ior, nodeColor=<rgba 0.7344 0 1 1>}
  NamedRerouteDeclaration(None=Divide(A=PixelDepth().R, B=NamedRerouteUsage())) {name=RealPixelDepth, nodeColor=<rgba 1 0 0.4922 1>}
  n5 = ViewProperty() {property=MEVP_TemporalSampleIndex}
  MaterialExpressionAdd_5 = Add(A=ScreenPosition().PixelPosition, B=AppendVector(A=@n5.Property, B=@n5.Property))
  n6 = Custom!(p=@MaterialExpressionAdd_5) {
   code="Mod( ((uint)(p.x) + 2 * (uint)(p.y)) , 5 )"
   outputType=CMOT_Float1
   inputs [{inputName=p, input {expression=&self:MaterialExpressionAdd_5}}]
  }
  n7 = TextureSample(UVs=MaterialFunctionCall!(TextureResolution=Constant2Vector() {g=64, r=64}) {materialFunction=/Engine/Functions/Engine_MaterialFunctions02/Texturing/ScreenAlignedPixelToPixelUVs}.Result) {
   samplerSource=SSM_Clamp_WorldGroupSettings
   samplerType=SAMPLERTYPE_LinearGrayscale
   texture=/Engine/EngineMaterials/Good64x64TilingNoiseHighFreq
  }
  NamedRerouteDeclaration(None=Divide(A=Add(A=@n6, B=@n7.R) {constB=0}) {constB=6}) {name=Temporalsample, nodeColor=<rgba 1 0 0.3984 1>}
  n8 = Constant()
  NamedRerouteDeclaration(None=Saturate(None=FunctionInput(Preview=@n8) {bUsePreviewValueAsDefault=true, inputName=BlurIntensity, inputType=FunctionInput_Scalar})) {name=BlurIntensity, nodeColor=<rgba 1 0 0.1875 1>}
  n9 = ComponentMask(None=ViewProperty() {property=MEVP_TanHalfFieldOfView}.InvProperty) {r=true}
  n10 = NamedRerouteUsage()
  n11 = Multiply(A=Multiply(A=ViewProperty() {property=MEVP_BufferSize}.InvProperty, B=Divide(A=@n9, B=@n10)), B=ComponentMask(None=ViewSize()) {r=true})
  n12 = Multiply(A=@n11, B=NamedRerouteUsage())
  n13 = Constant() {r=40}
  n14 = Subtract(A=NamedRerouteUsage())
  NamedRerouteDeclaration(None=Multiply(A=Multiply(A=@n12, B=@n13) {desc=Blur Multiplier}, B=Saturate(None=@n14))) {name=BlurRadius, nodeColor=<rgba 1 0 0.3984 1>}
  MaterialExpressionNamedRerouteUsage_17 = NamedRerouteUsage()
  n15 = Custom!(uv=@MaterialExpressionNamedRerouteUsage_17) {
   code="return BufferUVToViewportUV(uv);"
   description=BufferUVToViewportUV
   outputType=CMOT_Float2
   inputs [{inputName=uv, input {expression=&self:MaterialExpressionNamedRerouteUsage_17}}]
  }
  n16 = Subtract(A=Divide(A=SceneDepth!(OffsetFraction=@n15) {inputMode=OffsetFraction}.R, B=NamedRerouteUsage()), B=NamedRerouteUsage())
  n17 = LinearInterpolate(A=Constant() {r=500}, B=Constant() {r=1}, Alpha=NamedRerouteUsage()) {constA=500, constB=10}
  NamedRerouteDeclaration(None=Saturate(None=Divide(A=@n16, B=@n17) {constB=5000})) {name=DistanceBehind, nodeColor=<rgba 1 0.8906 0 1>}
  n18 = CameraVectorWS()
  NamedRerouteDeclaration(None=DotProduct(
    A=MaterialFunctionCall() {materialFunction=&CameraDirectionVector}.Result
    B=Multiply(A=@n18) {constB=-1}
   )) {name=cosdepth, nodeColor=<rgba 0.8047 0 1 1>}
  MaterialExpressionTransform_2 = Transform(None=CameraVectorWS()) {transformSourceType=TRANSFORMSOURCE_World, transformType=TRANSFORM_View}
  n19 = Constant3Vector() {constant=<rgba 0 0 1 0>}
  MaterialExpressionMultiply_10 = Multiply(
   A=Transform(None=FunctionInput(Preview=@n19) {bUsePreviewValueAsDefault=true, inputName=Normals, sortPriority=1}) {transformType=TRANSFORM_View}
   B=Constant3Vector() {constant=<rgba 1 -1 1 0>}
  )
  MaterialExpressionDivide_9 = Divide(B=NamedRerouteUsage()) {desc=air to glass}
  n20 = Custom!(Ray=@MaterialExpressionTransform_2, Normal=@MaterialExpressionMultiply_10, Index=@MaterialExpressionDivide_9) {
   code="return refract(Ray, Normal, Index);"
   description=Refract
   inputs [
    {inputName=Ray, input {expression=&self:MaterialExpressionTransform_2}}
    {inputName=Normal, input {expression=&self:MaterialExpressionMultiply_10}}
    {inputName=Index, input {expression=&self:MaterialExpressionDivide_9}}
   ]
  }
  n21 = Subtract(A=NamedRerouteUsage())
  n22 = Multiply(
   A=ComponentMask(None=@n20) {g=true, r=true}
   B=LinearInterpolate(B=ScalarParameter() {defaultValue=4, group=Refraction, parameterName=RefractionDepth, sliderMax=30, sliderMin=1}, Alpha=Saturate(None=@n21)) {constB=10}
  ) {desc=Refract Depth}
  NamedRerouteDeclaration(None=StaticSwitch(True=Constant2Vector(), False=Multiply(A=@n22, B=@n11), Value=NamedRerouteUsage())) {name=RefractionUVOffset, nodeColor=<rgba 0.0078 0 1 1>}
  NamedRerouteDeclaration(None=FunctionInput(Preview=StaticBool()) {bUsePreviewValueAsDefault=true, inputName=DisableRefraction, inputType=FunctionInput_StaticBool, sortPriority=5}) {name=DisableRefraction, nodeColor=<rgba 0 0.0859 1 1>}
 ]
}
