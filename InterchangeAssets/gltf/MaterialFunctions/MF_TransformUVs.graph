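# glTF import material function that transforms a selected UV channel (0–3) by scale, rotation in radians and offset.
# TexCoord picks the channel through nested Ifs by default, with a static switch to a Switch node that risks Nanite mesh cracks.
# A Custom node then applies per-axis wrap, clamp (pass-through) or mirror tiling.
# Outputs are tiled UVs, DDX/DDY taken before tiling, and the rotation's sin/cos.
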MaterialFunction /InterchangeAssets/gltf/MaterialFunctions/MF_TransformUVs
graph {nodes [
  n1 = NamedRerouteUsage()
  n2 = FunctionInput() {bUsePreviewValueAsDefault=true, inputName=Scale_X, inputType=FunctionInput_Scalar, sortPriority=2, previewValue=<1 0 0 0>}
  n3 = FunctionInput() {bUsePreviewValueAsDefault=true, inputName=Scale_Y, inputType=FunctionInput_Scalar, sortPriority=3, previewValue=<1 0 0 0>}
  n4 = AppendVector(A=@n2, B=@n3)
  n5 = FunctionInput() {bUsePreviewValueAsDefault=true, description=Rad, inputName=Rotation, inputType=FunctionInput_Scalar, sortPriority=4, previewValue=<0 0 0 0>}
  n6 = AppendVector(A=Sine(None=@n5), B=Cosine(None=@n5))
  n7 = FunctionInput() {bUsePreviewValueAsDefault=true, inputName=Offset_X, inputType=FunctionInput_Scalar, previewValue=<0 0 1 1>}
  n8 = FunctionInput() {bUsePreviewValueAsDefault=true, inputName=Offset_Y, inputType=FunctionInput_Scalar, sortPriority=1, previewValue=<0 0 1 1>}
  MaterialExpressionAdd_1 = Add(
   A=MaterialFunctionCall!(UV=Multiply(A=@n1, B=@n4), RotationCenter=Constant2Vector(), SinCosAngle=@n6) {materialFunction=&/MF_RotateV2}.Result
   B=AppendVector(A=@n7, B=@n8)
  )
  MaterialExpressionFunctionInput_6 = FunctionInput() {bUsePreviewValueAsDefault=true, inputName=TilingMethod, inputType=FunctionInput_Vector2, sortPriority=6}
  n9 = Custom!(UV=@MaterialExpressionAdd_1, TilingMethod=@MaterialExpressionFunctionInput_6) {
   code="float Result_U = 0;\r\nfloat Result_V = 0;\r\n\r\n//U component's tiling:\r\nif (TilingMethod.x < 1)\r\n{\r\n    //WRAP\r\n    //Texture sampler is set to Shared: Wrap\r\n    //do nothing\r\n    Result_U = fmod(UV.x, 1.0);\r\n    if (Result_U < 0) Result_U = 1.0 + Result_U;\r\n}\r\nelse if (TilingMethod.x < 2)\r\n{\r\n    //CLAMP\r\n    //Texture sampler is set to Shared: Clamp\r\n    //do nothing\r\n    Result_U = UV.x;\r\n}\r\nelse \r\n{\r\n    //MIRROR\r\n    float LerpValue_U = fmod(floor(abs(UV.x)), 2);\r\n    float ToLerp_U = fmod(abs(UV.x), 1);\r\n\r\n    Result_U = lerp(ToLerp_U, 1.0-ToLerp_U, LerpValue_U);\r\n}\r\n\r\n//V component's tiling:\r\nif (TilingMethod.y < 1)\r\n{\r\n    //WRAP\r\n    //Texture sampler is set to Shared: Wrap\r\n    //do nothing\r\n    Result_V = fmod(UV.y, 1.0);\r\n    if (Result_V < 0) Result_V = 1.0 + Result_V;\r\n}\r\nelse if (TilingMethod.y < 2)\r\n{\r\n    //CLAMP\r\n    //Texture sampler is set to Shared: Clamp\r\n    //do nothing\r\n    Result_V = UV.y;\r\n}\r\nelse \r\n{\r\n    //MIRROR\r\n    float LerpValue_V = fmod(floor(abs(UV.y)), 2);\r\n    float ToLerp_V = fmod(abs(UV.y), 1);\r\n\r\n    Result_V = lerp(ToLerp_V, 1.0-ToLerp_V, LerpValue_V);\r\n}\r\n\r\nreturn float2(Result_U, Result_V);"
   description=ApplyTiling
   outputType=CMOT_Float2
   showCode=true
   inputs [
    {inputName=UV, input {expression=&self:MaterialExpressionAdd_1}}
    {inputName=TilingMethod, input {expression=&self:MaterialExpressionFunctionInput_6}}
   ]
  }
  FunctionOutput(None=@n9) {outputName=UVs}
  FunctionOutput(None=DDX(Value=@MaterialExpressionAdd_1)) {outputName=DDX, sortPriority=1}
  FunctionOutput(None=DDY(Value=@MaterialExpressionAdd_1)) {outputName=DDY, sortPriority=2}
  FunctionOutput(None=@n6) {outputName=SinCos, sortPriority=3}
  NamedRerouteDeclaration(None=TextureCoordinate()) {name=UV0, nodeColor=<rgba 1 0.0938 0 1>}
  NamedRerouteDeclaration(None=TextureCoordinate() {coordinateIndex=1}) {name=UV1, nodeColor=<rgba 0 0.2266 1 1>}
  NamedRerouteDeclaration(None=TextureCoordinate() {coordinateIndex=2}) {name=UV2, nodeColor=<rgba 0.5469 1 0 1>}
  NamedRerouteDeclaration(None=TextureCoordinate() {coordinateIndex=3}) {name=UV3, nodeColor=<rgba 1 0 0.8672 1>}
  NamedRerouteDeclaration(None=FunctionInput() {bUsePreviewValueAsDefault=true, inputName=TexCoord, inputType=FunctionInput_Scalar, sortPriority=5}) {name=TexCoord, nodeColor=<rgba 0 1 0.8125 1>}
  n10 = NamedRerouteUsage()
  n11 = If(A=NamedRerouteUsage(), "A > B"=@n10, "A == B"=@n10, "A < B"=NamedRerouteUsage()) {constB=3}
  n12 = If(A=NamedRerouteUsage(), "A > B"=@n11, "A == B"=@n11, "A < B"=NamedRerouteUsage()) {constB=2}
  MaterialExpressionNamedRerouteUsage_1 = NamedRerouteUsage()
  MaterialExpressionNamedRerouteUsage_2 = NamedRerouteUsage()
  MaterialExpressionNamedRerouteUsage_3 = NamedRerouteUsage()
  MaterialExpressionNamedRerouteUsage_4 = NamedRerouteUsage()
  n13 = Switch!(
   SwitchValue=NamedRerouteUsage()
   Default=@MaterialExpressionNamedRerouteUsage_1
   UV0=@MaterialExpressionNamedRerouteUsage_1
   UV1=@MaterialExpressionNamedRerouteUsage_2
   UV2=@MaterialExpressionNamedRerouteUsage_3
   UV3=@MaterialExpressionNamedRerouteUsage_4
  ) {inputs [
    {
     inputName=UV0
     input {
      expression=&self:MaterialExpressionNamedRerouteUsage_1
      inputName=None
      mask=0
      maskA=0
      maskB=0
      maskG=0
      maskR=0
      outputIndex=0
     }
    }
    {
     inputName=UV1
     input {
      expression=&self:MaterialExpressionNamedRerouteUsage_2
      inputName=None
      mask=0
      maskA=0
      maskB=0
      maskG=0
      maskR=0
      outputIndex=0
     }
    }
    {
     inputName=UV2
     input {
      expression=&self:MaterialExpressionNamedRerouteUsage_3
      inputName=None
      mask=0
      maskA=0
      maskB=0
      maskG=0
      maskR=0
      outputIndex=0
     }
    }
    {
     inputName=UV3
     input {
      expression=&self:MaterialExpressionNamedRerouteUsage_4
      inputName=None
      mask=0
      maskA=0
      maskB=0
      maskG=0
      maskR=0
      outputIndex=0
     }
    }
   ]}
  n14 = StaticSwitchParameter(True=If(A=NamedRerouteUsage(), "A > B"=@n12, "A == B"=@n12, "A < B"=NamedRerouteUsage()) {constB=1}, False=@n13) {
   defaultValue=true
   desc="If true, use Switch expression in MF_TransformUVs, disabled by default because the Switch expression doesn't go to the Analytical Derivatives path that can leads to cracks on meshes."
   group=Nanite
   parameterName=DisableSwitchNaniteTransformUV
  }
  NamedRerouteDeclaration(None=@n14) {name=UV, nodeColor=<rgba 1 0.4922 0 1>}
 ]}
